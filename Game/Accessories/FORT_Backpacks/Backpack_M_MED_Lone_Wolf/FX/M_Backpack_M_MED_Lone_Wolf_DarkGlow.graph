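# Unlit, two-sided translucent Niagara sprite material for a soft, dark glow effect on the Lone Wolf backpack.
# Two radial sphere masks around the UV center separately shape a dim emissive and a stronger alpha, both modulated by particle color.
# Emissive passes through an FX adjustment function, while alpha fades with view angle.
# A shading path switch adds depth fade on the default path and omits it on mobile.
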Material /Game/Accessories/FORT_Backpacks/Backpack_M_MED_Lone_Wolf/FX/M_Backpack_M_MED_Lone_Wolf_DarkGlow
graph {
 properties {
  bUsedWithNiagaraSprites=true
  blendMode=BLEND_Translucent
  pixelDepthOffsetMode=PDOM_Legacy
  shadingModel=MSM_Unlit
  twoSided=true
  usedShadingModels=Unlit
 }
 nodes [
  n1 = TextureCoordinate()
  n2 = Constant2Vector() {g=0.5, r=0.5}
  n3 = ScalarParameter() {defaultValue=0.4, parameterName=Emissive Radius}
  n4 = ScalarParameter() {parameterName=Emissive Hardness}
  n5 = Multiply(
   A=Power(Base=SphereMask(A=@n1, B=@n2, Radius=@n3, Hardness=@n4), Exp=ScalarParameter() {defaultValue=2, parameterName=Emissive Power})
   B=ScalarParameter() {defaultValue=0.02, parameterName=Emissive Mul}
  )
  n6 = VectorParameter() {parameterName=Emissive Color, defaultValue=<rgba 1 1 1 1>}
  n7 = ParticleColor()
  MaterialFunctionCall(Emissive=Multiply(A=Multiply(A=@n5, B=@n6.RGB), B=@n7.RGB)) {materialFunction=&MF_FX_EmissiveAdjustments}
  n8 = ScalarParameter() {defaultValue=0.5, parameterName=Alpha Radius}
  n9 = ScalarParameter() {defaultValue=0.6, parameterName=Alpha Hardness}
  n10 = Multiply(
   A=Power(Base=SphereMask(A=@n1, B=@n2, Radius=@n8, Hardness=@n9), Exp=ScalarParameter() {defaultValue=1, parameterName=Alpha Power})
   B=ScalarParameter() {defaultValue=0.8, parameterName=Alpha Mul}
  )
  n11 = Multiply(
   A=Multiply(A=@n7.A, B=Saturate(None=@n10))
   B=MaterialFunctionCall("FadeSensitivity (optional)"=ScalarParameter() {defaultValue=0.25, group=Alpha, parameterName=FadeViewAngle}) {materialFunction=&FadeBasedOnViewAngle}.Result
  )
  n12 = ScalarParameter() {defaultValue=2, group=Alpha, parameterName=FadeDistance}
  ShadingPathSwitch(Default=Multiply(A=@n11, B=DepthFade(FadeDistance=@n12)), Mobile=@n11) {desc=Migrated from FeatureLevelSwitch}
 ]
}
